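# Physics asset for the Street Fashion Garnet backpack cosmetic, providing collision and constrained motion for its dangling headphones and zippers.
# The pack bone carries two capsules and anchors the headphones and both zipper pulls, while each earcup hangs off the headphones body.
# Headphone joints allow only a few degrees of movement, and zippers mainly twist, with swing almost locked.
# Shapes are simple capsules and boxes.
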PhysicsAsset /BRCosmetics/Accessories/FORT_Backpacks/Backpack_F_MED_Street_Fashion_Garnet/Meshes/F_MED_Street_Fashion_Garnet_Pack_Physics
bodies {
 dyn_headphones [
  {shapeType=Capsule, length=12.3361, radius=1.3347, center=<0.4632 0.1197 0.2078>, rotation=<rot 0.0097 -178.0421 -90.3426>}
  {shapeType=Box, extentX=1.4309, extentY=0.9001, extentZ=4.8742, center=<0.0001 -6.5043 -2.6628>, rotation=<rot 0 0 17.933>}
  {shapeType=Box, extentX=1.4925, extentY=0.8181, extentZ=5.4978, center=<0 6.0317 -2.2979>, rotation=<rot 0 0 -20.6704>}
 ]
 dyn_headphones_L [{shapeType=Capsule, length=2.1106, radius=1.7251, center=<0.1115 0.4819 -0.8694>, rotation=<rot -22.1342 -89.9999 -3.7325>}]
 dyn_headphones_R [{shapeType=Capsule, length=2.1559, radius=1.7732, center=<0.327 -0.6393 -1.1214>, rotation=<rot 22.1469 -90 -5.717>}]
 dyn_zipper_01 [{shapeType=Box, extentX=0.7462, extentY=0.5, extentZ=1.8877, center=<-0.1919 0 -0.6902>}]
 dyn_zipper_02 [{shapeType=Box, extentX=0.7462, extentY=0.5, extentZ=1.9361, center=<0 0 -0.5115>}]
 pack [
  {shapeType=Capsule, length=9.2425, radius=4.6054, center=<-0.0008 -5.1908 -0.9345>, rotation=<rot 0 89.9596 -89.8378>}
  {shapeType=Capsule, length=17.5, radius=2.75, center=<0.0001 -5.2491 -7.4381>, rotation=<rot 90 0.0198 0.0198>}
 ]
}
constraints [
 {bone1Name=dyn_headphones, bone2Name=pack, swing1LimitDegrees=0, swing2LimitDegrees=3, twistLimitDegrees=3}
 {bone1Name=dyn_headphones_L, bone2Name=dyn_headphones, swing1LimitDegrees=1, swing2LimitDegrees=0, twistLimitDegrees=3}
 {bone1Name=dyn_headphones_R, bone2Name=dyn_headphones, swing1LimitDegrees=1, swing2LimitDegrees=0, twistLimitDegrees=3}
 {bone1Name=dyn_zipper_01, bone2Name=pack, swing1LimitDegrees=1, swing2LimitDegrees=0, twistLimitDegrees=20}
 {bone1Name=dyn_zipper_02, bone2Name=pack, swing1LimitDegrees=1, swing2LimitDegrees=0, twistLimitDegrees=20}
]
